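# Physics asset for the ShovelCrop cape cosmetic on a medium female character, named as a set of cape colliders.
# It defines four bodies on the left and right clavicle and upper arm bones, covering the shoulder area.
# It has no constraints between the bodies, so it sets up no jointed ragdoll chain.
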PhysicsAsset /BRCosmetics/Accessories/FORT_Capes/Cape_F_MED_ShovelCrop/Meshes/F_MED_ShovelCrop_CapeColliders_Physics
bodies {clavicle_l [], clavicle_r [], upperarm_l [], upperarm_r []}
constraints []
