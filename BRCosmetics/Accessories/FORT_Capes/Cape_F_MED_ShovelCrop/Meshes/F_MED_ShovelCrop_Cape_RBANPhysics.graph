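# Physics asset for the ShovelCrop cape cosmetic, named for RigidBody anim node (RBAN) simulation.
# Center, left and right chains of six capsule bodies hang from MSTR bodies, with tight twist and swing limits that loosen toward the hem.
# Angularly free constraints tie the center chain to both sides at segments 04 and 06, linking the strips like a sheet.
# Capsules on torso, clavicles, upper arms and legs supply the character's body shapes.
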PhysicsAsset /BRCosmetics/Accessories/FORT_Capes/Cape_F_MED_ShovelCrop/Meshes/F_MED_ShovelCrop_Cape_RBANPhysics
bodies {
 calf_l [{shapeType=Capsule, length=32.779, radius=7.5, shapeName=calf_l_capsule, center=<-30.2881 -2.5633 0>, rotation=<rot 90 0 0>}]
 calf_r [{shapeType=Capsule, length=32.779, radius=7.5, shapeName=calf_r_capsule, center=<30.0763 2.5633 0>, rotation=<rot 90 180 0>}]
 clavicle_l [{shapeType=Capsule, length=17.0958, radius=6, shapeName=clavicle_l_capsule, center=<3.2325 -1.588 -0.6572>, rotation=<rot 90 0.0002 0>}]
 clavicle_r [{shapeType=Capsule, length=17.0958, radius=6, shapeName=clavicle_r_capsule, center=<-3.2325 1.588 0.6572>, rotation=<rot 90 -179.9998 0>}]
 dyn_cape_01 [{shapeType=Capsule, length=9.5, radius=0.5, shapeName=dyn_cape_01_capsule, center=<3.2243 0 4>, rotation=<rot 90 0 0>}]
 dyn_cape_02 [{shapeType=Capsule, length=10, radius=6, shapeName=dyn_cape_02_capsule, center=<3.2243 0 4>, rotation=<rot 90 0 0>}]
 dyn_cape_03 [{shapeType=Capsule, length=10, radius=6, shapeName=dyn_cape_03_capsule, center=<3.2243 0 4>, rotation=<rot 90 0 0>}]
 dyn_cape_04 [{shapeType=Capsule, length=9.5, radius=6.5, shapeName=dyn_cape_04_capsule, center=<3.2243 0 4>, rotation=<rot 90 0 0>}]
 dyn_cape_05 [{shapeType=Capsule, length=9.5, radius=6.5, shapeName=dyn_cape_05_capsule, center=<3.2243 0 4>, rotation=<rot 90 0 0>}]
 dyn_cape_06 [{shapeType=Capsule, length=9.5, radius=6.5, shapeName=dyn_cape_06_capsule, center=<3.2243 0 4>, rotation=<rot 90 0 0>}]
 dyn_cape_MSTR [{shapeType=Capsule, length=0.505, radius=1, shapeName=dyn_cape_MSTR_capsule, center=<0 -1.4272 0>, rotation=<rot 90 0 0>}]
 dyn_cape_l_01 [{shapeType=Capsule, length=9.5, radius=0.5, shapeName=dyn_cape_l_01_capsule, center=<3.2243 0 4>, rotation=<rot 90 0 0>}]
 dyn_cape_l_02 [{shapeType=Capsule, length=10, radius=6, shapeName=dyn_cape_l_02_capsule, center=<3.2243 0 4>, rotation=<rot 90 0 0>}]
 dyn_cape_l_03 [{shapeType=Capsule, length=10, radius=6, shapeName=dyn_cape_l_03_capsule, center=<3.2243 0 4>, rotation=<rot 90 0 0>}]
 dyn_cape_l_04 [{shapeType=Capsule, length=10, radius=6, shapeName=dyn_cape_l_04_capsule, center=<3.2243 0 4>, rotation=<rot 90 0 0>}]
 dyn_cape_l_05 [{shapeType=Capsule, length=9.5, radius=6.5, shapeName=dyn_cape_l_05_capsule, center=<3.2243 0 4>, rotation=<rot 90 0 0>}]
 dyn_cape_l_06 [{shapeType=Capsule, length=18.5, radius=6.5, shapeName=dyn_cape_l_06_capsule, center=<10.3939 0 4>, rotation=<rot 90 -90 0>}]
 dyn_cape_l_MSTR [{shapeType=Capsule, length=2.5, radius=3.5, shapeName=dyn_cape_l_MSTR_capsule, center=<-0.2029 0 4>, rotation=<rot 90 0 0>}]
 dyn_cape_r_01 [{shapeType=Capsule, length=9.5, radius=0.5, shapeName=dyn_cape_r_01_capsule, center=<3.2243 0 4>, rotation=<rot 90 0 0>}]
 dyn_cape_r_02 [{shapeType=Capsule, length=10, radius=6, shapeName=dyn_cape_r_02_capsule, center=<3.2243 0 4>, rotation=<rot 90 0 0>}]
 dyn_cape_r_03 [{shapeType=Capsule, length=10, radius=6, shapeName=dyn_cape_r_03_capsule, center=<3.2243 0 4>, rotation=<rot 90 0 0>}]
 dyn_cape_r_04 [{shapeType=Capsule, length=10, radius=6, shapeName=dyn_cape_r_04_capsule, center=<3.2243 0 4>, rotation=<rot 90 0 0>}]
 dyn_cape_r_05 [{shapeType=Capsule, length=9.5, radius=6.5, shapeName=dyn_cape_r_05_capsule, center=<3.2243 0 4>, rotation=<rot 90 0 0>}]
 dyn_cape_r_06 [{shapeType=Capsule, length=18, radius=6.5, shapeName=dyn_cape_r_06_capsule, center=<10.6751 0 -0.1508>, rotation=<rot 90 -90 0>}]
 dyn_cape_r_MSTR [{shapeType=Capsule, length=2.5, radius=3.5, shapeName=dyn_cape_r_MSTR_capsule, center=<-0.2029 0 4>, rotation=<rot 90 0 0>}]
 pelvis [{shapeType=Capsule, length=16.4087, radius=11, shapeName=pelvis_capsule}]
 spine_01 [{shapeType=Capsule, length=18, radius=8.5, shapeName=spine_01_capsule, center=<0 1.2785 0>}]
 spine_02 [{shapeType=Capsule, length=11.5, radius=7, shapeName=spine_02_capsule, center=<0 1.4758 0>}]
 spine_03 [{shapeType=Capsule, length=16.4036, radius=6.5, shapeName=spine_03_capsule, center=<0.0169 1.489 0>}]
 spine_04 [{shapeType=Capsule, length=12.005, radius=8.5, shapeName=spine_04_capsule, center=<0 0.7246 0>}]
 spine_05 [{shapeType=Capsule, length=23.4036, radius=7, shapeName=spine_05_capsule, center=<3.3119 -0.8306 0>}]
 thigh_l [{shapeType=Capsule, length=24.1204, radius=9.5, shapeName=thigh_l_capsule, center=<-18.7785 0.0905 -0.1067>, rotation=<rot 90 0 0>}]
 thigh_r [{shapeType=Capsule, length=24.1204, radius=9.5, shapeName=thigh_r_capsule, center=<18.7785 -0.0905 0.1067>, rotation=<rot 90 180 0>}]
 upperarm_l [{shapeType=Capsule, length=21.8974, radius=5.4004, shapeName=upperarm_l_capsule, center=<9.6486 -1.0279 -0.4832>, rotation=<rot 90 180 0>}]
 upperarm_r [{shapeType=Capsule, length=21.8974, radius=5.4004, shapeName=upperarm_r_capsule, center=<-9.6486 1.0279 0.4832>, rotation=<rot 90 0 0>}]
}
constraints [
 {bone1Name=dyn_cape_01, bone2Name=dyn_cape_MSTR, swing1LimitDegrees=10, swing2LimitDegrees=10, twistLimitDegrees=1}
 {bone1Name=dyn_cape_02, bone2Name=dyn_cape_01, swing1LimitDegrees=20, swing2LimitDegrees=10, twistLimitDegrees=1}
 {bone1Name=dyn_cape_03, bone2Name=dyn_cape_02, swing1LimitDegrees=20, swing2LimitDegrees=20, twistLimitDegrees=1}
 {bone1Name=dyn_cape_04, bone2Name=dyn_cape_03, swing1LimitDegrees=20, swing2LimitDegrees=20, twistLimitDegrees=1}
 {bone1Name=dyn_cape_05, bone2Name=dyn_cape_04, swing1LimitDegrees=45, swing2LimitDegrees=30, twistLimitDegrees=1}
 {bone1Name=dyn_cape_06, bone2Name=dyn_cape_05, swing1LimitDegrees=45, swing2LimitDegrees=30, twistLimitDegrees=1}
 {bone1Name=dyn_cape_l_01, bone2Name=dyn_cape_l_MSTR, swing1LimitDegrees=10, swing2LimitDegrees=10, twistLimitDegrees=1}
 {bone1Name=dyn_cape_l_02, bone2Name=dyn_cape_l_01, swing1LimitDegrees=20, swing2LimitDegrees=10, twistLimitDegrees=1}
 {bone1Name=dyn_cape_l_03, bone2Name=dyn_cape_l_02, swing1LimitDegrees=20, swing2LimitDegrees=20, twistLimitDegrees=1}
 {bone1Name=dyn_cape_l_04, bone2Name=dyn_cape_l_03, swing1LimitDegrees=20, swing2LimitDegrees=20, twistLimitDegrees=1}
 {bone1Name=dyn_cape_l_05, bone2Name=dyn_cape_l_04, swing1LimitDegrees=45, swing2LimitDegrees=30, twistLimitDegrees=1}
 {bone1Name=dyn_cape_l_06, bone2Name=dyn_cape_l_05, swing1LimitDegrees=45, swing2LimitDegrees=30, twistLimitDegrees=1}
 {bone1Name=dyn_cape_r_01, bone2Name=dyn_cape_r_MSTR, swing1LimitDegrees=10, swing2LimitDegrees=10, twistLimitDegrees=1}
 {bone1Name=dyn_cape_r_02, bone2Name=dyn_cape_r_01, swing1LimitDegrees=20, swing2LimitDegrees=10, twistLimitDegrees=1}
 {bone1Name=dyn_cape_r_03, bone2Name=dyn_cape_r_02, swing1LimitDegrees=20, swing2LimitDegrees=20, twistLimitDegrees=1}
 {bone1Name=dyn_cape_r_04, bone2Name=dyn_cape_r_03, swing1LimitDegrees=20, swing2LimitDegrees=20, twistLimitDegrees=1}
 {bone1Name=dyn_cape_r_05, bone2Name=dyn_cape_r_04, swing1LimitDegrees=45, swing2LimitDegrees=30, twistLimitDegrees=1}
 {bone1Name=dyn_cape_r_06, bone2Name=dyn_cape_r_05, swing1LimitDegrees=45, swing2LimitDegrees=30, twistLimitDegrees=1}
 {
  bone1Name=dyn_cape_04
  bone2Name=dyn_cape_r_04
  swing1LimitDegrees=45
  swing1Motion=Free
  swing2LimitDegrees=45
  swing2Motion=Free
  twistLimitDegrees=45
  twistMotion=Free
 }
 {
  bone1Name=dyn_cape_04
  bone2Name=dyn_cape_l_04
  swing1LimitDegrees=45
  swing1Motion=Free
  swing2LimitDegrees=45
  swing2Motion=Free
  twistLimitDegrees=45
  twistMotion=Free
 }
 {
  bone1Name=dyn_cape_06
  bone2Name=dyn_cape_r_06
  swing1LimitDegrees=45
  swing1Motion=Free
  swing2LimitDegrees=45
  swing2Motion=Free
  twistLimitDegrees=45
  twistMotion=Free
 }
 {
  bone1Name=dyn_cape_06
  bone2Name=dyn_cape_l_06
  swing1LimitDegrees=45
  swing1Motion=Free
  swing2LimitDegrees=45
  swing2Motion=Free
  twistLimitDegrees=45
  twistMotion=Free
 }
]
